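# Niagara camera effect for the Helios open-desert wasteland, layering on-screen dirt, wind-blown landscape dust and near-ground particles around the viewer.
# DirtOnScreen and NearDirt fill boxes near the owner, with the former's pulsing spawn rate scaled by facing against world X.
# Dust, Skulls and SnowflakesGround layers follow terrain via landscape sampling or height texture, with dust and skulls fading by player distance.
# All five GPU sprite emitters cull spawns inside four fixed world-space exclusion boxes.
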NiagaraSystem /FNGameplayCues/Environments/Helios/FX/OpenDesert/FX/NS_Wasteland_CameraFX
dependencies {
 usedDataInterfaces [
  &NiagaraDataInterfaceCurve
  &NiagaraDataInterfaceCollisionQuery
  &NiagaraDataInterfaceColorCurve
  &NiagaraDataInterfaceSpriteRendererInfo
 ]
 usedDynamicInputs []
 usedModules [&SystemState]
}
emitters {
 DirtOnScreen {
  simTarget=GPUComputeSim
  emitterUpdateScript [
   &EmitterState()
   &SpawnRate(
    SpawnRate=&Multiply_Float(
     A=125
     B=&Multiply_Float(
      A=&Lerp_Float(
       A=0
       B=2
       Alpha=&ClampFloat(Float=&DotProduct(Vector A=<1 0 0>, Vector B=$Engine.Owner.SystemXAxis))
      )
      B=&Waveform("Waveform [1]"=::NewEnumerator12, "[1] Frequency"=1, "[1] Amplitude Min/Max"=<0 1>)
     )
    )
    Use Spawn Probability=false
   )
  ]
  particleSpawnScript [
   &InitializeParticle(
    Color=<rgba 1 1 1 0.35>
    Lifetime Max=5
    Lifetime Min=0.75
    Lifetime Mode=::NewEnumerator1
    Sprite Rotation Mode=::NewEnumerator2
    Sprite Size Mode=::NewEnumerator1
    Uniform Sprite Size Max=5
    Uniform Sprite Size Min=2
    UsePositionOffset=false
    Write Material Random=true
   )
   &ShapeLocation(
    Box Midpoint=<0.5 0.5 0.9>
    Box Size=<10 240 450>
    MaskSpawnGroup=false
    Override Seed=false
    Shape Primitive=::NewEnumerator2
    Surface Only=false
    UseOwnerScale=true
   )
   &RegionMask(Region Radius=75, UseOffset=true, UseRotation=false)
   &KillParticlesInVolume(
    Box Size:Vector3f=<15000 4500 3500>
    Kill Shape=::NewEnumerator1
    Origin Offset:Vector3f=<-107457.3594 -47112.4688 3249.8899>
    Volume Origin:Vector3f=&AddVectorToPosition(Position=<0 0 0>, Vector=<0 0 0>)
   )
   &KillParticlesInVolume(
    Box Size:Vector3f=<4500 7500 6500>
    Kill Shape=::NewEnumerator1
    Origin Offset:Vector3f=<-77582.7422 -7489.02 2260.7>
    Volume Origin:Vector3f=&AddVectorToPosition(Position=<0 0 0>, Vector=<0 0 0>)
   ) {module=KillParticlesInVolume003}
   &KillParticlesInVolume(
    Box Size:Vector3f=<9750 14500 6500>
    Kill Shape=::NewEnumerator1
    Origin Offset:Vector3f=<-44833.3633 -15931.1836 2260.7>
    Volume Origin:Vector3f=&AddVectorToPosition(Position=<0 0 0>, Vector=<0 0 0>)
   ) {module=KillParticlesInVolume002}
   &KillParticlesInVolume(
    Box Size:Vector3f=<6500 6500 6500>
    Kill Shape=::NewEnumerator1
    Origin Offset:Vector3f=<-101205.2578 18688.6699 3249.8899>
    Volume Origin:Vector3f=&AddVectorToPosition(Position=<0 0 0>, Vector=<0 0 0>)
   ) {module=KillParticlesInVolume001}
  ]
  particleUpdateScript [
   &ParticleState()
   &ScaleSpriteSize(
    Scale Sprite Size Mode=::NewEnumerator0
    Uniform Scale Factor=&Lerp_Float(A=1.4, B=0.3, Alpha=$StackContext.RegionMask.Float)
   ) {module=ScaleSpriteSize001}
   &ScaleColor(
    Scale Alpha=&FloatFromCurve(
     FloatCurve=DataInterface() {propertyValues {Curve {keys [<at 0 1 RCIM_Cubic>, {arriveTangent=-0.5488, interpMode=RCIM_Cubic, leaveTangent=-0.5488, tangentMode=RCTM_User, time=1, value=0}]}}}
     Scale Curve=&Multiply_Float(
      A=&FloatFromCurve(FloatCurve=DataInterface() {propertyValues {Curve {keys [
           {interpMode=RCIM_Cubic, tangentMode=RCTM_User, time=0, value=1}
           {arriveTangent=-1, interpMode=RCIM_Cubic, leaveTangent=-1, time=0.9991, value=0.9936}
           {arriveTangent=-3.5405, interpMode=RCIM_Cubic, leaveTangent=-3.5405, tangentMode=RCTM_User, time=1, value=0}
          ]}}})
      B=&FloatFromCurve(
       FloatCurve=DataInterface() {propertyValues {Curve {keys [
           {arriveTangent=0.0119, interpMode=RCIM_Cubic, leaveTangent=0.0119, tangentMode=RCTM_User, time=0, value=0}
           {arriveTangent=3.7909, interpMode=RCIM_Cubic, leaveTangent=3.7909, tangentMode=RCTM_User, time=0.1588, value=0.1618}
           {interpMode=RCIM_Cubic, tangentMode=RCTM_User, time=1, value=1}
          ]}}}
       Scale Curve=&FloatFromCurve(
        FloatCurve=DataInterface() {propertyValues {Curve {keys [
            {arriveTangent=4.1785, interpMode=RCIM_Cubic, leaveTangent=4.1785, tangentMode=RCTM_User, time=0, value=0}
            {arriveTangent=1.1179, interpMode=RCIM_Cubic, leaveTangent=1.1179, tangentMode=RCTM_User, time=0.2042, value=0.6228}
            {arriveTangent=0.0224, interpMode=RCIM_Cubic, leaveTangent=0.0224, tangentMode=RCTM_User, time=1, value=1}
           ]}}}
        CurveIndex=$Engine.System.Age
       )
      )
     )
    )
    ScaleA=true
    ScaleRGB=true
   )
   &GravityForce(Gravity=<0 0 0>)
   &SolveForcesAndVelocity(Clamp Velocity=false, Limit Acceleration=false)
  ]
  rendererClasses [&NiagaraSpriteRendererProperties]
  renderers [{rendererClass=&NiagaraSpriteRendererProperties}]
 }
 Dust_Landscape {
  simTarget=GPUComputeSim
  emitterUpdateScript [
   &EmitterState()
   &SpawnRate(SpawnRate=50, Use Spawn Probability=false)
  ]
  particleSpawnScript [
   &AddVelocity(Rotation Coordinate Space=::World, Velocity=<50 0 0>)
   &InitializeParticle(
    Color Maximum=<rgba 1 0.2567 0.0925 0.5>
    Color Minimum=<rgba 1 0.2778 0.13 0.15>
    Color Mode=::NewEnumerator2
    Lifetime Max=3
    Lifetime Min=2
    Lifetime Mode=::NewEnumerator1
    Mesh Scale Mode=::NewEnumerator1
    Mesh Uniform Scale Max=10
    Mesh Uniform Scale Min=5
    Sprite Rotation Mode=::NewEnumerator2
    Sprite Size Mode=::NewEnumerator1
    Sprite UV Mode=::NewEnumerator1
    Uniform Sprite Size Max=1000
    Uniform Sprite Size Min=900
    UsePositionOffset=false
    Write Material Random=true
   )
   /SetVariables() {isSetParameters=true, module=SetVariables_#1}
   &ShapeLocation(
    Cylinder Radius=3500
    MaskSpawnGroup=false
    Override Seed=false
    Shape Primitive=::NewEnumerator1
    Surface Only=false
    Surface Only Band Thickness=0.8
    UseOwnerScale=true
   )
   &self:Sample_Landscape_01() {module=Sample_Landscape_01}
   /SetVariables(Particles.New Vector 3f:Vector3f=<0 0 1>) {isSetParameters=true, module=SetVariables_#2}
   &DynamicMaterialParameters(
    Index 0 Param 1=&RandomRangeFloat(Minimum=50, Maximum=100)
    Index 0 Param 2=5
    Index 0 Param 3=15
    Index 0 Param 4=&RandomRangeFloat(Minimum=0, Maximum=0)
    Index 1 Param 1=50
    Index 2 Param 1=&RandomRangeFloat(Minimum=-0.1, Maximum=-0.1)
    Index 2 Param 2=&RandomRangeFloat(Minimum=0.1, Maximum=0)
    Index 2 Param 3=&RandomRangeFloat(Minimum=0, Maximum=0)
    Index 2 Param 4=&RandomRangeFloat(Minimum=0.2, Maximum=0.3)
    Module.Index0Param0Write=true
    Module.Index0Param1Write=true
    Module.Index0Param2Write=true
    Module.Index0Param3Write=true
    Module.Index1Param0Write=true
    Module.Index1Param1Write=true
    Module.Index1Param2Write=true
    Module.Index1Param3Write=true
    Module.Index2Param0Write=true
    Module.Index2Param1Write=true
    Module.Index2Param2Write=true
    Module.Index2Param3Write=true
    Write Parameter Index 1=true
    Write Parameter Index 2=true
   )
   /SetVariables(Particles.Linear Color=<rgba 1 0 0 1>, Particles.WaveScale:Vector3f=<6 6 6>) {isSetParameters=true, module=SetVariables_#3}
   &self:Rejection_Sample_-_Normal(B=0.5, Landscape:Vector3f=$Particles.Landscape Normal, Spawn Vector:Vector3f=<0 0 1>) {module=Rejection_Sample_-_Normal}
   &ApplyInitialForces()
   &KillParticlesInVolume(
    Box Size:Vector3f=<15000 4500 3500>
    Kill Shape=::NewEnumerator1
    Origin Offset:Vector3f=<-107457.3594 -47112.4688 3249.8899>
    Volume Origin:Vector3f=&AddVectorToPosition(Position=<0 0 0>, Vector=<0 0 0>)
   )
   &KillParticlesInVolume(
    Box Size:Vector3f=<4500 7500 6500>
    Kill Shape=::NewEnumerator1
    Origin Offset:Vector3f=<-77582.7422 -7489.02 2260.7>
    Volume Origin:Vector3f=&AddVectorToPosition(Position=<0 0 0>, Vector=<0 0 0>)
   ) {module=KillParticlesInVolume003}
   &KillParticlesInVolume(
    Box Size:Vector3f=<9750 14500 6500>
    Kill Shape=::NewEnumerator1
    Origin Offset:Vector3f=<-44833.3633 -15931.1836 2260.7>
    Volume Origin:Vector3f=&AddVectorToPosition(Position=<0 0 0>, Vector=<0 0 0>)
   ) {module=KillParticlesInVolume002}
   &KillParticlesInVolume(
    Box Size:Vector3f=<6500 6500 6500>
    Kill Shape=::NewEnumerator1
    Origin Offset:Vector3f=<-101205.2578 18688.6699 3249.8899>
    Volume Origin:Vector3f=&AddVectorToPosition(Position=<0 0 0>, Vector=<0 0 0>)
   ) {module=KillParticlesInVolume001}
  ]
  particleUpdateScript [
   &self:Sample_Landscape_UPdate_01(Bias:NiagaraFloat=$Particles.Terrain Offset, New Landscape Sample:NiagaraDataInterfaceLandscape=$User.Landscape) {module=Sample_Landscape_UPdate_01}
   &ParticleState()
   &ScaleSpriteSize(Uniform Curve Index=&self:ScratchDynamicInput_01(
     End Distance:NiagaraFloat=7500
     Max Output:NiagaraFloat=1
     MCP Player Location:LinearColor=$NPC.FortniteNPC.PlayerPosition
     Minimum Output:NiagaraFloat=0
     Start Distance:NiagaraFloat=0
    )) {module=ScaleSpriteSize001}
   &SpriteRotationRate(Rotation Rate=-60)
   &ScaleSpriteSize(Uniform Curve Sprite Scale=DataInterface() {propertyValues {Curve {keys [{interpMode=RCIM_Cubic, leaveTangent=3.9356, tangentMode=RCTM_Break, time=0, value=0}, <at 1 1 RCIM_Cubic>]}}})
   &WindForce(Scale=0, Wind Speed=<750 0 0>, Wind Speed Scale=&RandomRangeFloat(Minimum=0.5, Maximum=1))
   &Drag(Drag=4, Use Linear Drag=true, Use Rotational Drag=true)
   &SolveForcesAndVelocity(Clamp Velocity=false, Limit Acceleration=false)
   &ScaleSpriteSizeBySpeed(Max Scale Factor=<2 12>, Min Scale Factor=<0.05 0.05>)
   &RegionMask(
    Invert Region=true
    Region Dimensions:Vector3f=<15000 15000 15000>
    Region Falloff=1
    Region Type:ENiagaraDebugDrawShape=::NewEnumerator1
    UseOffset=true
    UseRotation=false
   )
   &ScaleColor(
    Scale Alpha=&FloatFromCurve(
     FloatCurve=DataInterface() {propertyValues {Curve {keys [
         {interpMode=RCIM_Cubic, tangentMode=RCTM_User, time=0, value=0}
         {interpMode=RCIM_Cubic, tangentMode=RCTM_User, time=0.5, value=1}
         {tangentMode=RCTM_User, time=1, value=0}
        ]}}}
     Scale Curve=&Multiply_Float(A=0.5, B=$StackContext.RegionMask.Float)
    )
    Scale RGB=&Multiply_VectorByFloat(Vector=<1 1 1>, Float=&RandomRangeFloat(Minimum=1, Maximum=2))
    ScaleA=true
    ScaleRGB=true
   ) {module=ScaleColor001}
   /SetVariables(
    Particles.Linear Color=&MultiplyLinearColor(
     Linear Color=$Particles.Color
     Scale Factor=&MakeVector4_FromFloat(Float=&FloatFromCurve(
       FloatCurve=DataInterface() {propertyValues {Curve {keys [
           {interpMode=RCIM_Cubic, tangentMode=RCTM_User, time=0, value=0}
           {interpMode=RCIM_Cubic, tangentMode=RCTM_User, time=0.5, value=1}
           {tangentMode=RCTM_User, time=1, value=0}
          ]}}}
       Scale Curve=$System.ColorIntensityMult
      ))
    )
    Particles.Linear Color001:LinearColor=&MultiplyLinearColor(
     Linear Color=&ColorFromCurve(ColorCurve=DataInterface() {propertyValues {
        AlphaCurve {keys [<at 0 0 RCIM_Cubic>,<at 0.1218 1>,<at 1 0 RCIM_Cubic>]}
        BlueCurve {keys [<at 0 0.0688 RCIM_Cubic>,<at 1 1 RCIM_Cubic>]}
        GreenCurve {keys [<at 0 1 RCIM_Cubic>,<at 1 1 RCIM_Cubic>]}
        RedCurve {keys [<at 0 0 RCIM_Cubic>,<at 1 1 RCIM_Cubic>]}
       }})
     Scale Factor=&MakeVector4_FromFloat(Float=&FloatFromCurve(
       FloatCurve=DataInterface() {propertyValues {Curve {keys [<at 0 1>,<at 1 0>]}}}
       Scale Curve=&Multiply_Float(A=$System.ColorIntensityMult, B=4)
      ))
    )
   ) {isSetParameters=true, module=SetVariables_#4}
   &CameraOffset(Camera Offset Amount=10) {module=CameraOffset001}
   &SubUVAnimation(
    "Sprite Renderer "=DataInterface() {propertyValues {SpriteRenderer=&self:Dust_Landscape_7.NiagaraSpriteRendererProperties_1}}
    UseEndFrame=false
    UseStartFrame=false
   )
   &KillParticles(Kill Particles=&Boolean_AndOperation(
     Bool1=&SetBoolByFloatComparison(
      A=&MakeFloatFromVector4(Vector 4=&MakeVector4_FromLinearColor(LinearColor=$Particles.Color), Channel=::NewEnumerator3)
      B=0.0003
      Comparison Type=::NewEnumerator6
     )
     Bool2=&SetBoolByFloatComparison(A=0.5, B=$Particles.NormalizedAge)
    )) {enabled=false}
   &ScaleColor(
    Scale Alpha=&Lerp_Float(
     A=0
     B=1
     Alpha=&self:ScratchDynamicInput_01(
      End Distance:NiagaraFloat=3500
      Max Output:NiagaraFloat=1
      MCP Player Location:LinearColor=$NPC.FortniteNPC.PlayerPosition
      Minimum Output:NiagaraFloat=0
      Start Distance:NiagaraFloat=0
     )
    )
    ScaleA=true
    ScaleRGB=true
   )
  ]
  rendererClasses [&NiagaraSpriteRendererProperties]
  renderers [
   {rendererClass=&NiagaraSpriteRendererProperties}
   {rendererClass=&NiagaraSpriteRendererProperties, rendererIndex=1}
  ]
 }
 NearDirt {
  simTarget=GPUComputeSim
  emitterUpdateScript [
   &EmitterState()
   &SpawnRate(
    SpawnRate=&Multiply_Float(
     A=250
     B=&FloatFromCurve(
      FloatCurve=DataInterface() {propertyValues {Curve {keys [
          {interpMode=RCIM_Cubic, tangentMode=RCTM_Break, time=0, value=0}
          {interpMode=RCIM_Cubic, tangentMode=RCTM_Break, time=0.2, value=0.85}
          {arriveTangent=0.125, interpMode=RCIM_Cubic, leaveTangent=0.125, time=0.4, value=0}
          <at 0.6 0.9 RCIM_Cubic>
          {arriveTangent=0.25, interpMode=RCIM_Cubic, leaveTangent=0.25, time=0.8, value=0}
          <at 1 1>
         ]}}}
      CurveIndex=$Emitter.NormalizedLoopAge
     )
    )
    Use Spawn Probability=false
   )
  ]
  particleSpawnScript [
   &InitializeParticle(
    Color=&RandomRangeLinearColor(Minimum=<rgba 0.5 0.1313 0.0885 0.5>, Maximum=<rgba 1.25 0.5535 0.4948 0.5>)
    Lifetime Max=3
    Lifetime Min=0.75
    Lifetime Mode=::NewEnumerator1
    Sprite Rotation Mode=::NewEnumerator2
    Sprite Size Mode=::NewEnumerator1
    Uniform Sprite Size Max=4
    Uniform Sprite Size Min=1
    UsePositionOffset=false
    Write Material Random=true
   )
   &ShapeLocation(
    Box Midpoint=<0.5 0.5 0.9>
    Box Size=<0 240 500>
    MaskSpawnGroup=false
    Override Seed=false
    Shape Primitive=::NewEnumerator2
    Surface Only=false
    UseOwnerScale=true
   )
   &RegionMask(Region Radius=250, UseOffset=true, UseRotation=false)
   &KillParticlesInVolume(
    Box Size:Vector3f=<6500 6500 6500>
    Kill Shape=::NewEnumerator1
    Origin Offset:Vector3f=<-101205.2578 18688.6699 3249.8899>
    Volume Origin:Vector3f=&AddVectorToPosition(Position=<0 0 0>, Vector=<0 0 0>)
   ) {module=KillParticlesInVolume001}
   &KillParticlesInVolume(
    Box Size:Vector3f=<9750 14500 6500>
    Kill Shape=::NewEnumerator1
    Origin Offset:Vector3f=<-44833.3633 -15931.1836 2260.7>
    Volume Origin:Vector3f=&AddVectorToPosition(Position=<0 0 0>, Vector=<0 0 0>)
   ) {module=KillParticlesInVolume002}
   &KillParticlesInVolume(
    Box Size:Vector3f=<4500 7500 6500>
    Kill Shape=::NewEnumerator1
    Origin Offset:Vector3f=<-77582.7422 -7489.02 2260.7>
    Volume Origin:Vector3f=&AddVectorToPosition(Position=<0 0 0>, Vector=<0 0 0>)
   ) {module=KillParticlesInVolume003}
   &KillParticlesInVolume(
    Box Size:Vector3f=<15000 4500 3500>
    Kill Shape=::NewEnumerator1
    Origin Offset:Vector3f=<-107457.3594 -47112.4688 3249.8899>
    Volume Origin:Vector3f=&AddVectorToPosition(Position=<0 0 0>, Vector=<0 0 0>)
   )
  ]
  particleUpdateScript [
   &ParticleState()
   &ScaleSpriteSize()
   &ScaleColor(
    Scale Alpha=&FloatFromCurve(
     FloatCurve=DataInterface() {propertyValues {Curve {keys [
         {tangentMode=RCTM_User, time=0, value=1}
         {arriveTangent=-0.0405, interpMode=RCIM_Cubic, leaveTangent=-0.0405, tangentMode=RCTM_User, time=0.436, value=0.6364}
         {arriveTangent=-0.5488, interpMode=RCIM_Cubic, leaveTangent=-0.5488, tangentMode=RCTM_User, time=1, value=0}
        ]}}}
     Scale Curve=&Multiply_Float(
      A=&FloatFromCurve(FloatCurve=DataInterface() {propertyValues {Curve {keys [
           {interpMode=RCIM_Cubic, tangentMode=RCTM_User, time=0, value=1}
           {arriveTangent=-1, interpMode=RCIM_Cubic, leaveTangent=-1, time=0.9991, value=0.9936}
           {arriveTangent=-3.5405, interpMode=RCIM_Cubic, leaveTangent=-3.5405, tangentMode=RCTM_User, time=1, value=0}
          ]}}})
      B=&FloatFromCurve(
       FloatCurve=DataInterface() {propertyValues {Curve {keys [
           {arriveTangent=0.0119, interpMode=RCIM_Cubic, leaveTangent=0.0119, tangentMode=RCTM_User, time=0, value=0}
           {arriveTangent=3.7909, interpMode=RCIM_Cubic, leaveTangent=3.7909, tangentMode=RCTM_User, time=0.1588, value=0.1618}
           {interpMode=RCIM_Cubic, tangentMode=RCTM_User, time=1, value=1}
          ]}}}
       Scale Curve=&FloatFromCurve(
        FloatCurve=DataInterface() {propertyValues {Curve {keys [
            {arriveTangent=4.1785, interpMode=RCIM_Cubic, leaveTangent=4.1785, tangentMode=RCTM_User, time=0, value=0}
            {arriveTangent=1.1179, interpMode=RCIM_Cubic, leaveTangent=1.1179, tangentMode=RCTM_User, time=0.2042, value=0.6228}
            {arriveTangent=0.0224, interpMode=RCIM_Cubic, leaveTangent=0.0224, tangentMode=RCTM_User, time=1, value=1}
           ]}}}
        CurveIndex=$Engine.System.Age
       )
      )
     )
    )
    ScaleA=true
    ScaleRGB=true
   )
   &GravityForce(Gravity=<500 0 0>)
   &CurlNoiseForce(Noise Frequency=150, Noise Strength=150, Pan Noise=false)
   &SolveForcesAndVelocity(Clamp Velocity=false, Limit Acceleration=false)
   &ScaleSpriteSizeBySpeed(Max Scale Factor=<5 15>, Min Scale Factor=<0.05 0.05>)
  ]
  rendererClasses [&NiagaraSpriteRendererProperties]
  renderers [{rendererClass=&NiagaraSpriteRendererProperties}]
 }
 Skulls_Landscape001 {
  simTarget=GPUComputeSim
  emitterUpdateScript [
   &EmitterState()
   &SpawnRate(SpawnRate=50, Use Spawn Probability=false)
  ]
  particleSpawnScript [
   &AddVelocity(Rotation Coordinate Space=::World, Velocity=<150 0 0>)
   &InitializeParticle(
    Color Maximum=<rgba 1 0.3035 0.2479 0.5>
    Color Minimum=<rgba 0.5417 0.2393 0.127 0.12>
    Color Mode=::NewEnumerator2
    Lifetime Max=3
    Lifetime Min=2
    Lifetime Mode=::NewEnumerator1
    Mesh Scale Mode=::NewEnumerator1
    Mesh Uniform Scale Max=10
    Mesh Uniform Scale Min=5
    Sprite Rotation Mode=::NewEnumerator2
    Sprite Size Mode=::NewEnumerator1
    Sprite UV Mode=::NewEnumerator1
    Uniform Sprite Size Max=500
    Uniform Sprite Size Min=450
    UsePositionOffset=false
    Write Material Random=true
   )
   /SetVariables(Particles.Terrain Offset=50) {isSetParameters=true, module=SetVariables_#1}
   &ShapeLocation(
    Cylinder Radius=750
    MaskSpawnGroup=false
    Override Seed=false
    Shape Primitive=::NewEnumerator1
    Surface Only=true
    Surface Only Band Thickness=0.8
    Use Endcaps In Surface Only Mode=false
    UseOwnerScale=true
   )
   &self:Sample_Landscape_03() {module=Sample_Landscape_03}
   /SetVariables(Particles.New Vector 3f:Vector3f=<0 0 1>) {isSetParameters=true, module=SetVariables_#2}
   &DynamicMaterialParameters(
    Index 0 Param 1=&RandomRangeFloat(Minimum=50, Maximum=100)
    Index 0 Param 2=5
    Index 0 Param 3=15
    Index 0 Param 4=&RandomRangeFloat(Minimum=0, Maximum=0)
    Index 1 Param 1=50
    Index 2 Param 1=&RandomRangeFloat(Minimum=-0.1, Maximum=-0.1)
    Index 2 Param 2=&RandomRangeFloat(Minimum=0.1, Maximum=0)
    Index 2 Param 3=&RandomRangeFloat(Minimum=0, Maximum=0)
    Index 2 Param 4=&RandomRangeFloat(Minimum=0.2, Maximum=0.3)
    Module.Index0Param0Write=true
    Module.Index0Param1Write=true
    Module.Index0Param2Write=true
    Module.Index0Param3Write=true
    Module.Index1Param0Write=true
    Module.Index1Param1Write=true
    Module.Index1Param2Write=true
    Module.Index1Param3Write=true
    Module.Index2Param0Write=true
    Module.Index2Param1Write=true
    Module.Index2Param2Write=true
    Module.Index2Param3Write=true
    Write Parameter Index 1=true
    Write Parameter Index 2=true
   )
   /SetVariables(Particles.Linear Color=<rgba 1 0 0 1>, Particles.WaveScale:Vector3f=<6 6 6>) {isSetParameters=true, module=SetVariables_#3}
   &self:Rejection_Sample_-_Normal_02(B=0.5, Landscape:Vector3f=$Particles.Landscape Normal, Spawn Vector:Vector3f=<1 0 1>) {module=Rejection_Sample_-_Normal_02}
   &ApplyInitialForces()
   &KillParticles(Kill Particles=&SetBoolByFloatComparison(
     A=&ClampFloat(Float=&DotProduct(Vector A=$Engine.Owner.SystemXAxis, Vector B=<-1 0 0>))
     B=0.1
    )) {module=KillParticles001}
   &KillParticlesInVolume(
    Box Size:Vector3f=<15000 4500 3500>
    Kill Shape=::NewEnumerator1
    Origin Offset:Vector3f=<-107457.3594 -47112.4688 3249.8899>
    Volume Origin:Vector3f=&AddVectorToPosition(Position=<0 0 0>, Vector=<0 0 0>)
   )
   &KillParticlesInVolume(
    Box Size:Vector3f=<4500 7500 6500>
    Kill Shape=::NewEnumerator1
    Origin Offset:Vector3f=<-77582.7422 -7489.02 2260.7>
    Volume Origin:Vector3f=&AddVectorToPosition(Position=<0 0 0>, Vector=<0 0 0>)
   ) {module=KillParticlesInVolume003}
   &KillParticlesInVolume(
    Box Size:Vector3f=<9750 14500 6500>
    Kill Shape=::NewEnumerator1
    Origin Offset:Vector3f=<-44833.3633 -15931.1836 2260.7>
    Volume Origin:Vector3f=&AddVectorToPosition(Position=<0 0 0>, Vector=<0 0 0>)
   ) {module=KillParticlesInVolume002}
   &KillParticlesInVolume(
    Box Size:Vector3f=<6500 6500 6500>
    Kill Shape=::NewEnumerator1
    Origin Offset:Vector3f=<-101205.2578 18688.6699 3249.8899>
    Volume Origin:Vector3f=&AddVectorToPosition(Position=<0 0 0>, Vector=<0 0 0>)
   ) {module=KillParticlesInVolume001}
  ]
  particleUpdateScript [
   &self:Sample_Landscape_UPdate_03(Bias:NiagaraFloat=$Particles.Terrain Offset, New Landscape Sample:NiagaraDataInterfaceLandscape=$User.Landscape) {module=Sample_Landscape_UPdate_03}
   &ParticleState()
   &SpriteRotationRate(Rotation Rate=-60)
   &ScaleSpriteSize(Uniform Curve Sprite Scale=DataInterface() {propertyValues {Curve {keys [{interpMode=RCIM_Cubic, leaveTangent=3.9356, tangentMode=RCTM_Break, time=0, value=0}, <at 1 1 RCIM_Cubic>]}}})
   &WindForce(Scale=0, Wind Speed=<750 0 0>, Wind Speed Scale=&RandomRangeFloat(Minimum=0.5, Maximum=1))
   &Drag(Drag=8, Use Linear Drag=true, Use Rotational Drag=true)
   &SolveForcesAndVelocity(Clamp Velocity=false, Limit Acceleration=false)
   &ScaleSpriteSizeBySpeed(Max Scale Factor=<2 6>, Min Scale Factor=<0.05 0.05>)
   &RegionMask(
    Invert Region=true
    Region Dimensions:Vector3f=<500 500 500>
    Region Falloff=1
    Region Type:ENiagaraDebugDrawShape=::NewEnumerator1
    UseOffset=true
    UseRotation=false
   )
   &ScaleColor(
    Scale Alpha=&FloatFromCurve(
     FloatCurve=DataInterface() {propertyValues {Curve {keys [
         {interpMode=RCIM_Cubic, tangentMode=RCTM_User, time=0, value=0}
         {interpMode=RCIM_Cubic, tangentMode=RCTM_User, time=0.5, value=1}
         {tangentMode=RCTM_User, time=1, value=0}
        ]}}}
     Scale Curve=&Multiply_Float(A=0.2, B=$StackContext.RegionMask.Float)
    )
    Scale RGB=&Multiply_VectorByFloat(Vector=<1 1 1>, Float=&RandomRangeFloat(Minimum=1, Maximum=2))
    ScaleA=true
    ScaleRGB=true
   ) {module=ScaleColor001}
   /SetVariables(
    Particles.Linear Color=&MultiplyLinearColor(
     Linear Color=$Particles.Color
     Scale Factor=&MakeVector4_FromFloat(Float=&FloatFromCurve(
       FloatCurve=DataInterface() {propertyValues {Curve {keys [
           {interpMode=RCIM_Cubic, tangentMode=RCTM_User, time=0, value=0}
           {interpMode=RCIM_Cubic, tangentMode=RCTM_User, time=0.5, value=1}
           {tangentMode=RCTM_User, time=1, value=0}
          ]}}}
       Scale Curve=$System.ColorIntensityMult
      ))
    )
    Particles.Linear Color001:LinearColor=&MultiplyLinearColor(
     Linear Color=&ColorFromCurve(ColorCurve=DataInterface() {propertyValues {
        AlphaCurve {keys [<at 0 0 RCIM_Cubic>,<at 0.1218 1>,<at 1 0 RCIM_Cubic>]}
        BlueCurve {keys [<at 0 0.0688 RCIM_Cubic>,<at 1 1 RCIM_Cubic>]}
        GreenCurve {keys [<at 0 1 RCIM_Cubic>,<at 1 1 RCIM_Cubic>]}
        RedCurve {keys [<at 0 0 RCIM_Cubic>,<at 1 1 RCIM_Cubic>]}
       }})
     Scale Factor=&MakeVector4_FromFloat(Float=&FloatFromCurve(
       FloatCurve=DataInterface() {propertyValues {Curve {keys [<at 0 1>,<at 1 0>]}}}
       Scale Curve=&Multiply_Float(A=$System.ColorIntensityMult, B=4)
      ))
    )
   ) {isSetParameters=true, module=SetVariables_#4}
   &CameraOffset(Camera Offset Amount=10) {module=CameraOffset001}
   &SubUVAnimation(
    "Sprite Renderer "=DataInterface() {propertyValues {SpriteRenderer=&self:Skulls_Landscape001_1.NiagaraSpriteRendererProperties_1}}
    UseEndFrame=false
    UseStartFrame=false
   )
   &KillParticles(Kill Particles=&Boolean_AndOperation(
     Bool1=&SetBoolByFloatComparison(
      A=&MakeFloatFromVector4(Vector 4=&MakeVector4_FromLinearColor(LinearColor=$Particles.Color), Channel=::NewEnumerator3)
      B=0.0003
      Comparison Type=::NewEnumerator6
     )
     Bool2=&SetBoolByFloatComparison(A=0.5, B=$Particles.NormalizedAge)
    )) {enabled=false}
   &ScaleSpriteSize(
    Uniform Curve Index=$Particles.Age
    Uniform Curve Sprite Scale=DataInterface() {propertyValues {Curve {keys [<at 0.2 0>, {tangentMode=RCTM_User, time=1, value=1}]}}}
   ) {module=ScaleSpriteSize002}
   &ScaleSpriteSize(Uniform Curve Index=&self:ScratchDynamicInput_03(
     End Distance:NiagaraFloat=1500
     Max Output:NiagaraFloat=1
     MCP Player Location:LinearColor=$NPC.FortniteNPC.PlayerPosition
     Minimum Output:NiagaraFloat=0
     Start Distance:NiagaraFloat=0
    )) {module=ScaleSpriteSize001}
   &ScaleColor(
    Scale Alpha=&Lerp_Float(
     A=0
     B=1
     Alpha=&self:ScratchDynamicInput_03(
      End Distance:NiagaraFloat=500
      Max Output:NiagaraFloat=1
      MCP Player Location:LinearColor=$NPC.FortniteNPC.PlayerPosition
      Minimum Output:NiagaraFloat=0
      Start Distance:NiagaraFloat=0
     )
    )
    ScaleA=true
    ScaleRGB=true
   )
  ]
  rendererClasses [&NiagaraSpriteRendererProperties]
  renderers [{rendererClass=&NiagaraSpriteRendererProperties}]
 }
 SnowflakesGround {
  simTarget=GPUComputeSim
  emitterUpdateScript [
   &EmitterState()
   &SpawnRate(SpawnRate=&Multiply_Float(A=350, B=1), Use Spawn Probability=false)
  ]
  particleSpawnScript [
   /Niagara/Modules/Spawn/Initialization/InitializeParticle(
    Lifetime=&UniformRangedFloat(Minimum=2, Maximum=3)
    Module.Write Color=true
    Module.Write Lifetime=true
    Module.Write Mass=true
    Module.Write Material Random=true
    Module.Write Position=false
    Module.Write Scale=false
    Module.Write SpriteRotation=true
    Module.Write SpriteSize=true
    Sprite Rotation=&UniformRangedFloat(Minimum=0, Maximum=360)
    Sprite Size=&Vector2DFromFloat(Value=&UniformRangedFloat(Minimum=5, Maximum=7))
   ) {enabled=false}
   /Niagara/Modules/Spawn/Initialization/V2/InitializeParticle(
    Color=<rgba 1 0.3309 0.2969 0.5>
    Lifetime Max=3
    Lifetime Min=2
    Lifetime Mode=::NewEnumerator1
    Sprite Size Mode=::NewEnumerator1
    Uniform Sprite Size Min=6
    UsePositionOffset=false
    Write Material Random=true
   ) {module=InitializeParticle001}
   /Niagara/Modules/Spawn/Location/SphereLocation(
    MaskSpawnGroup=false
    Module.Override Randomness=false
    Non Uniform Scale=<1 1 0>
    Sphere Radius=0
    Surface Only=false
    Surface Only Band Thickness=1
   )
   &CylinderLocation(
    Cylinder Height=0
    Cylinder Radius=3500
    Lathe Profile=DataInterface() {propertyValues {Curve {keys []}}}
    MaskSpawnGroup=false
    Module.Enable Lathe Profile=false
    Module.Override Randomness=false
    Module.Surface Only=true
    Use Endcaps In Surface Only Mode=false
   )
   /NiagaraFN/Modules/Spawn/FN_SpawnParticlesOnTerrainNearPlayer(
    Height Texture -x -y:Vector3f=<-130048 -130048 0>
    HeightOffset=50
    HeightTexture +x +y:Vector3f=<130048 130048 0>
    Player Position:NiagaraPosition=$Particles.Position
    Player Relative Particle Location:Vector2f=&UniformRangedVector2D(Minimum=<0 0>, Maximum=<0 0>)
   ) {module=FN_SpawnParticlesOnTerrainNearPlayer001}
   &AddVelocity(Velocity=&UniformRangedVector(Minimum=<600 -256 100>, Maximum=<1200 256 150>)) {enabled=false}
   /Niagara/Modules/Update/Forces/CurlNoiseForce(Noise Frequency=25, Noise Strength=50, Pan Noise=true)
   &ApplyInitialForces()
   &KillParticlesInVolume(
    Box Size:Vector3f=<15000 4500 3500>
    Kill Shape=::NewEnumerator1
    Origin Offset:Vector3f=<-107457.3594 -47112.4688 3249.8899>
    Volume Origin:Vector3f=&AddVectorToPosition(Position=<0 0 0>, Vector=<0 0 0>)
   )
   &KillParticlesInVolume(
    Box Size:Vector3f=<4500 7500 6500>
    Kill Shape=::NewEnumerator1
    Origin Offset:Vector3f=<-77582.7422 -7489.02 2260.7>
    Volume Origin:Vector3f=&AddVectorToPosition(Position=<0 0 0>, Vector=<0 0 0>)
   ) {module=KillParticlesInVolume003}
   &KillParticlesInVolume(
    Box Size:Vector3f=<9750 14500 6500>
    Kill Shape=::NewEnumerator1
    Origin Offset:Vector3f=<-44833.3633 -15931.1836 2260.7>
    Volume Origin:Vector3f=&AddVectorToPosition(Position=<0 0 0>, Vector=<0 0 0>)
   ) {module=KillParticlesInVolume002}
   &KillParticlesInVolume(
    Box Size:Vector3f=<6500 6500 6500>
    Kill Shape=::NewEnumerator1
    Origin Offset:Vector3f=<-101205.2578 18688.6699 3249.8899>
    Volume Origin:Vector3f=&AddVectorToPosition(Position=<0 0 0>, Vector=<0 0 0>)
   ) {module=KillParticlesInVolume001}
  ]
  particleUpdateScript [
   &ParticleState() {module=UpdateAge}
   &ScaleColor(
    Scale Alpha=&FloatFromCurve(FloatCurve=DataInterface() {propertyValues {Curve {keys [
         <at -0.0007 -0.0138 RCIM_Cubic>
         {arriveTangent=2.6721, interpMode=RCIM_Cubic, leaveTangent=2.6721, tangentMode=RCTM_User, time=0.3127, value=0.3952}
         {arriveTangent=-0.5712, interpMode=RCIM_Cubic, leaveTangent=-0.5712, tangentMode=RCTM_User, time=0.5424, value=1.0086}
         <at 1.0011 0.002 RCIM_Cubic>
        ]}}})
    ScaleA=true
    ScaleRGB=true
   )
   &WindForce(Alignment Falloff Start / End:Vector2f=<0 1500>, Friction Distance Limit=55, Normal Friction Mask=0.5, Scale=0.1, Wind Speed=<1500 0 0>)
   &Drag(Drag=4, Use Linear Drag=true, Use Rotational Drag=true)
   &SolveForcesAndVelocity(Clamp Velocity=false, Limit Acceleration=false)
   &SpriteRotationRate(Rotation Rate=&UniformRangedFloat(Minimum=-15, Maximum=25))
   &ScaleSpriteSizeBySpeed(Max Scale Factor=<2 5>)
   &CameraOffset(Camera Offset Amount=-150)
  ]
  rendererClasses [&NiagaraSpriteRendererProperties]
  renderers [{rendererClass=&NiagaraSpriteRendererProperties}]
 }
}
summary {
 systemName=NS_Wasteland_CameraFX
 emitters [
  {emitterName=DirtOnScreen, simTarget=GPUComputeSim, rendererClasses [&NiagaraSpriteRendererProperties]}
  {emitterName=Dust_Landscape, simTarget=GPUComputeSim, rendererClasses [&NiagaraSpriteRendererProperties]}
  {emitterName=SnowflakesGround, simTarget=GPUComputeSim, rendererClasses [&NiagaraSpriteRendererProperties]}
  {emitterName=Skulls_Landscape001, simTarget=GPUComputeSim, rendererClasses [&NiagaraSpriteRendererProperties]}
  {emitterName=NearDirt, simTarget=GPUComputeSim, rendererClasses [&NiagaraSpriteRendererProperties]}
 ]
 userVariables [
  User.bIsOnWater {type=NiagaraBool}
  User.Landscape {
   type=NiagaraDataInterfaceLandscape
   description=""
   defaultValue {
    struct=NiagaraExt_VariableValue_DataInterface
    value {
     dataInterface=&self:NiagaraDataInterfaceLandscape_0
     dataInterfaceClass=&NiagaraDataInterfaceLandscape
    }
   }
  }
 ]
}
